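# Physics asset for the one-handed Pineapple Bandit pickaxe skeletal mesh, covering its handle and a dangling pineapple keychain.
# A capsule body covers the Melee_TwoHanded_Handle bone, and three more capsules cover the keychain's base, parent and end bones.
# Two constraints chain the keychain bodies base to parent to end, with moderate swing limits, slightly wider at the end, and very tight twist.
# No constraint links the keychain base to the handle body.
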PhysicsAsset /BRCosmetics/Weapons/FORT_Melee/Pickaxe_PineappleBandit_1h/Meshes/SK_Pickaxe_PineappleBandit_1h_Physics
bodies {
 Melee_TwoHanded_Handle [{shapeType=Capsule, length=45.8973, radius=4.2384, center=<0.0011 0.1785 26.8427>, rotation=<rot 0.3809 -90 0.0019>}]
 dyn_pineapple_keychain_01 [{shapeType=Capsule, length=1.5296, radius=4.3455, center=<4.7551 0 0>, rotation=<rot -90 0 0>}]
 dyn_pineapple_keychain_base [{shapeType=Capsule, length=3.0938, radius=1.1562, rotation=<rot 90 0 0>}]
 dyn_pineapple_keychain_parent [{shapeType=Capsule, length=1.625, radius=1.125, center=<2.9275 0 0>, rotation=<rot -90 0 0>}]
}
constraints [
 {
  bone1Name=dyn_pineapple_keychain_parent
  bone2Name=dyn_pineapple_keychain_base
  swing1LimitDegrees=20
  swing2LimitDegrees=40
  twistLimitDegrees=4
 }
 {
  bone1Name=dyn_pineapple_keychain_01
  bone2Name=dyn_pineapple_keychain_parent
  swing1LimitDegrees=30
  swing2LimitDegrees=45
  twistLimitDegrees=4
 }
]
